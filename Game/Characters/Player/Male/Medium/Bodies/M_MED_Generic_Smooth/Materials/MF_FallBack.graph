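# Material function for a "FallBack" player body look: a dark surface with glowing animated smoke, glitch flicker and a phase-out opacity mask.
# Pre-skinned local-position coordinates drive warped, chromatically offset smoke and vertical gradient masks, layered with fresnel rims and a timed flash into emissive.
# Sine-based flicker modulates emissive and, with world-aligned noise and a sphere pulse, horizontal vertex offset.
# Only the input material's normal is reused, blended with panning noise.
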MaterialFunction /Game/Characters/Player/Male/Medium/Bodies/M_MED_Generic_Smooth/Materials/MF_FallBack
graph {nodes [
  n1 = VectorParameter() {group=FallBack, parameterName=FallB_Base Color, sortPriority=1, defaultValue=<rgba 0.001 0.001 0.001 0.001>}
  n2 = Fresnel(ExponentIn=ScalarParameter() {defaultValue=5, group=FallBack Chrom, parameterName=CromaticExponent_C}) {baseReflectFraction=0, exponent=1}
  n3 = ScalarParameter() {defaultValue=5, group=FallBack Chrom, parameterName=Cromatic_C_thickness}
  n4 = Power(Base=OneMinus(None=Multiply(A=@n2, B=@n3)))
  n5 = Multiply(
   A=MaterialFunctionCall(Hue Shift Percentage=Saturate(None=@n4), Texture=Constant3Vector() {constant=<rgba 1 0 0 0>}) {materialFunction=&HueShift}.Result
   B=Fresnel(ExponentIn=ScalarParameter() {defaultValue=5, group=FallBack Chrom, parameterName=CromaticExponent_B}) {baseReflectFraction=0}
  )
  n6 = ScalarParameter() {group=FallBack Chrom, parameterName=Cromatic saturation}
  n7 = Time() {period=5}
  n8 = ScalarParameter() {group=FallBack Arsenic Flash, parameterName=FlashStartTime}
  n9 = Clamp(None=Divide(A=Subtract(A=@n7, B=@n8), B=ScalarParameter() {defaultValue=0.5, group=FallBack Arsenic Flash, parameterName=FlashDuration}))
  n10 = ScalarParameter() {group=FallBack, parameterName=FallB_Z_Masking_Gradent_background_Value, sortPriority=150}
  n11 = PreSkinnedLocalBounds()
  n12 = Subtract(A=VertexInterpolator(VS=PreSkinnedPosition()).PS, B=@n11.Min)
  n13 = Subtract(A=@n11.Max, B=@n11.Max)
  n14 = Divide(A=@n12, B=@n13)
  n15 = Divide(
   A=Add(A=@n14, B=ScalarParameter() {group=FallBack, parameterName=FallB_Z_MaskingA_Gradent_Offset, sortPriority=100})
   B=ScalarParameter() {defaultValue=1, group=FallBack, parameterName=FallB_Z_MaskingA_Gradent_Falloff, sortPriority=100}
  )
  n16 = ComponentMask(None=Power(Base=@n15)) {b=true}
  n17 = Divide(
   A=Add(A=@n14, B=ScalarParameter() {group=FallBack, parameterName=FallB_Z_MaskingB_Gradent_Offset, sortPriority=150})
   B=ScalarParameter() {defaultValue=1, group=FallBack, parameterName=FallB_Z_MaskingB_Gradent_Falloff, sortPriority=150}
  )
  n18 = ComponentMask(None=Power(Base=@n17)) {b=true}
  n19 = Multiply(
   A=LinearInterpolate(
    A=ScalarParameter() {group=FallBack, parameterName=FallB_Z_Masking_GradentA_A_Value, sortPriority=100}
    B=ScalarParameter() {defaultValue=1, group=FallBack, parameterName=FallB_Z_Masking_GradentA_B_Value, sortPriority=100}
    Alpha=Saturate(None=@n16)
   )
   B=LinearInterpolate(
    A=ScalarParameter() {defaultValue=1, group=FallBack, parameterName=FallB_Z_Masking_GradentB_A_Value, sortPriority=150}
    B=ScalarParameter() {defaultValue=1, group=FallBack, parameterName=FallB_Z_Masking_GradentB_B_Value, sortPriority=150}
    Alpha=Saturate(None=@n18)
   )
  )
  n20 = Multiply(
   A=VectorParameter() {group=FallBack, parameterName="FallB_Emissive_Base_Color ", defaultValue=<rgba 0 0 0 1>}.RGB
   B=Saturate(None=Add(A=@n19, B=@n10))
  )
  n21 = ActorPositionWS()
  n22 = MaterialFunctionCall(Float2=Round(None=Divide(A=@n21) {constB=2000})) {materialFunction=&BreakOutFloat2Components}
  n23 = Add(A=Time(), B=Add(A=@n22.R, B=@n22.R))
  n24 = Multiply(A=ScalarParameter() {defaultValue=3, group=FallBack_Glitch, parameterName=FlickerSpeed}, B=@n23)
  n25 = Sine(None=@n24) {period=1.6}
  n26 = Sine(None=@n24) {period=1.42}
  n27 = Ceil(None=AppendVector(A=AppendVector(A=@n25, B=@n26), B=Sine(None=@n24) {period=1.51}))
  n28 = ScalarParameter() {defaultValue=0.5, group=FallBack_Glitch, parameterName=NoiseFrequency}
  n29 = Add(A=Sine(None=@n24) {period=0.72}, B=Multiply(B=@n28) {constA=0.75}) {constB=0.75}
  n30 = Multiply(A=Add(A=Sine(None=@n24) {period=2.1}, B=Multiply(B=@n28) {constA=0.25}), B=@n29)
  n31 = Multiply(B=@n28) {constA=0.9, constB=0.9}
  n32 = Sine(None=@n24) {period=1.25}
  n33 = Multiply(A=@n29, B=Ceil(None=Add(A=@n31, B=@n32) {constB=0.9}))
  n34 = Sine(None=@n23) {period=0.075}
  n35 = Multiply(A=Saturate(None=Multiply(A=@n30, B=@n33)), B=LinearInterpolate(Alpha=ConstantBiasScale(None=@n34)) {constA=0.7})
  n36 = Add(
   A=ScalarParameter() {group=FallBack_Glitch, parameterName=NoiseAdd}
   B=Multiply(A=Multiply(A=@n27, B=@n35), B=ScalarParameter() {defaultValue=2, group=FallBack_Glitch, parameterName=NoiseMult})
  )
  n37 = Saturate(None=Desaturation(None=@n36))
  n38 = LinearInterpolate(B=ScalarParameter() {defaultValue=5, group=FallBack_Glitch, parameterName=FallB_Emissive_Base_Color_Glitch_Amount}, Alpha=@n37) {constA=1}
  n39 = VectorParameter() {group=FallBack, parameterName=FallB_Emissive_Smoke_Color_B, defaultValue=<rgba 0.5592 0.0683 1 1>}
  n40 = ScalarParameter() {defaultValue=0.5, group=FallBack, parameterName=FallB_Smoke_Warp_Amount}
  n41 = MaterialFunctionCall(Color=@n13) {materialFunction=&SplitComponents}
  n42 = Divide(A=@n12, B=Max(A=Max(A=@n41.R, B=@n41.R), B=@n41.B))
  n43 = ComponentMask(None=@n42) {b=true, r=true}
  n44 = ScalarParameter() {defaultValue=0.12, group=FallBack, parameterName=FallB_Smoke_Rotation}
  n45 = Time()
  n46 = ComponentMask(None=VectorParameter() {group=FallBack, parameterName=FallB_Smoke_Warp_SpeedB, defaultValue=<rgba -0.035 0.03 -0.1 0>}.RGB) {b=true, r=true}
  n47 = Multiply(
   A=Add(
    A=MaterialFunctionCall(UVs=@n43, "Rotation Angle (0-1)"=@n44) {materialFunction=&CustomRotator}.Rotated Values
    B=Multiply(A=@n45, B=@n46)
   )
   B=ScalarParameter() {defaultValue=1.75, group=FallBack, parameterName=FallB_Smoke_Warp_Tile_RateB}
  )
  n48 = TextureObject() {samplerType=SAMPLERTYPE_LinearColor, texture=/Game/Characters/Player/Male/Medium/Bodies/M_MED_Generic_Smooth/textures/T_FallBack_FX_M_1}
  n49 = Add(A=ComponentMask(None=TextureSample(UVs=@n47, Tex=@n48).A) {r=true}) {constB=-0.5}
  n50 = Add(A=@n14, B=ScalarParameter() {defaultValue=-1.15, group=FallBack, parameterName=FallB_Warp_Masking_Gradent_Offset})
  n51 = ScalarParameter() {defaultValue=-0.8, group=FallBack, parameterName=FallB_Warp_Masking_Gradent_Falloff}
  n52 = ComponentMask(None=Power(Base=Divide(A=@n50, B=@n51))) {b=true, r=true}
  n53 = Saturate(None=ComponentMask(None=MaterialFunctionCall(UVs=@n52, "Rotation Angle (0-1)"=@n44) {materialFunction=&CustomRotator}.Rotated Values) {g=true})
  n54 = MaterialFunctionCall(
   UVs=ComponentMask(None=@n42) {b=true, r=true}
   "Rotation Angle (0-1)"=ScalarParameter() {defaultValue=0.12, group=FallBack, parameterName=FallB_Smoke_Streams_Rotation}
  ) {materialFunction=&CustomRotator}
  n55 = Multiply(A=Time(), B=ComponentMask(None=VectorParameter() {group=FallBack, parameterName=FallB_Smoke Streams_Speed}.RGB) {b=true, r=true})
  n56 = VectorParameter() {group=FallBack, parameterName=FallB_Smoke Streams offset, defaultValue=<rgba 0.2571 0 0 1>}
  n57 = Multiply(
   A=Add(A=Add(A=@n54.Rotated Values, B=@n55), B=ComponentMask(None=@n56.RGB) {g=true, r=true})
   B=ComponentMask(None=VectorParameter() {group=FallBack, parameterName=FallB_Smoke Streams Tile Rate, defaultValue=<rgba 1.5 2 2 1>}.RGB) {g=true, r=true}
  )
  n58 = Add(A=@n57, B=Multiply(A=@n53, B=Multiply(A=@n40, B=@n49)))
  n59 = TextureObject() {samplerType=SAMPLERTYPE_Masks, texture=/Game/Characters/Player/Male/Medium/Bodies/M_MED_Generic_Smooth/textures/T_FallBack_FX_M}
  n60 = ComponentMask(None=@n42) {b=true, r=true}
  n61 = Time()
  n62 = ComponentMask(None=VectorParameter() {group=FallBack, parameterName=FallB_Smoke_Warp_Speed, defaultValue=<rgba -0.035 0.03 -0.1 0>}.RGB) {b=true, r=true}
  n63 = Multiply(
   A=Add(
    A=MaterialFunctionCall(UVs=@n60, "Rotation Angle (0-1)"=@n44) {materialFunction=&CustomRotator}.Rotated Values
    B=Multiply(A=@n61, B=@n62)
   )
   B=ScalarParameter() {defaultValue=2, group=FallBack, parameterName=FallB_Smoke_Warp_Tile_Rate}
  )
  n64 = Add(A=ComponentMask(None=TextureSample(UVs=@n63, Tex=@n48).A) {r=true}) {constB=-0.5}
  n65 = Add(A=@n57, B=Multiply(A=Multiply(A=@n40, B=@n64), B=@n53)) {constB=0}
  n66 = ComponentMask(None=@n65) {r=true}
  n67 = ComponentMask(None=@n65) {g=true}
  n68 = AppendVector(A=Add(A=@n66) {constB=0.025}, B=@n67)
  n69 = AppendVector(A=Add(A=@n66) {constB=0.05}, B=@n67)
  n70 = Multiply(
   A=MaterialFunctionCall(
    X=TextureSample(UVs=@n65, Tex=@n59).R
    Y=TextureSample(UVs=@n68, Tex=@n59) {texture=/Game/Global/Textures/Default/Caustics/T_Turbulence_Seamless}.R
    Z=TextureSample(UVs=@n69, Tex=@n59) {texture=&T_Turbulence_Seamless}.R
   ) {materialFunction=&MakeFloat3}.Result
   B=ScalarParameter() {defaultValue=1, group=FallBack Chrom, parameterName=FallB_Emissive_Smoke_Color_A_Chromatic}
  )
  n71 = VectorParameter() {group=FallBack, parameterName=FallB_Emissive_Smoke_Color_A_Flicker, defaultValue=<rgba 0.5592 0.0683 1 1>}
  n72 = LinearInterpolate(
   A=VectorParameter() {group=FallBack, parameterName=FallB_Emissive_Smoke_Color_A, defaultValue=<rgba 0.5592 0.0683 1 1>}.RGB
   B=@n71.RGB
   Alpha=@n37
  ) {constA=1, constB=2}
  n73 = Add(
   A=LinearInterpolate(A=Multiply(A=@n20, B=@n38), B=Multiply(A=@n39.RGB, B=@n19), Alpha=TextureSample(UVs=@n58, Tex=@n59).A)
   B=Multiply(A=Multiply(A=@n70, B=@n72), B=@n19)
  )
  n74 = VectorParameter() {group=FallBack, parameterName=FallB_Emissive_Dark_Fresnel_Color, defaultValue=<rgba 0 0 0 1>}
  n75 = ScalarParameter() {defaultValue=0.8401, group=FallBack, parameterName=FallB_Dark_Fresnel_Expo}
  n76 = Saturate(None=MaterialFunctionCall(
    In=Fresnel(ExponentIn=@n75) {baseReflectFraction=0}
    Contrast=ScalarParameter() {defaultValue=1, group=FallBack, parameterName=FallB_Dark_Fresnel_Contrast}
   ) {materialFunction=&CheapContrast}.Result)
  n77 = Add(
   A=Multiply(
    A=MaterialFunctionCall!(In=@n9) {materialFunction=/Game/Effects/Fort_Effects/Materials/_Functions/MF_Smoothstep_EaseOut}.Result
    B=VectorParameter() {group=FallBack Arsenic Flash, parameterName=FlashColor, defaultValue=<rgba 0.1258 0.0113 0.14 0>}.RGB
   )
   B=Multiply(
    A=LinearInterpolate(A=@n73, B=@n74.RGB, Alpha=@n76)
    B=ScalarParameter() {defaultValue=1, group=FallBack, parameterName=FallB_Emissive_Master_Brightness}
   )
  )
  n78 = Add(
   A=Multiply(
    A=VectorParameter() {group=FallBack Chrom, parameterName=Cromatic_Hue, defaultValue=<rgba 0 0 0 1>}.RGB
    B=Desaturation(None=@n5, Fraction=@n6)
   )
   B=@n77
  )
  n79 = Fresnel(
   ExponentIn=ScalarParameter() {defaultValue=1, group=FallBack EgdeGlow, parameterName=Fresnel Power, sortPriority=2}
   BaseReflectFractionIn=Constant()
  ) {baseReflectFraction=0}
  n80 = ScalarParameter() {defaultValue=1, group=FallBack EgdeGlow, parameterName=Fresnel Thicken, sortPriority=2}
  n81 = Saturate(None=Power(
    Base=Multiply(A=@n79, B=@n80)
    Exp=ScalarParameter() {defaultValue=1, group=FallBack EgdeGlow, parameterName=Fresnel PowerExp, sortPriority=2}
   ))
  n82 = VectorParameter() {group=FallBack EgdeGlow, parameterName=Fresnel Color, sortPriority=0, defaultValue=<rgba 0.3558 0 0.4 0>}
  n83 = Add(
   A=@n77
   B=Multiply(
    A=Multiply(A=@n81, B=@n82.RGB)
    B=ScalarParameter() {defaultValue=1, group=FallBack EgdeGlow, parameterName=Fresnel Blend Amount, sortPriority=-2}
   )
  )
  n84 = TextureObject() {samplerType=SAMPLERTYPE_Normal, texture=&T_Weap_Distort_N_02}
  n85 = Constant() {r=1}
  n86 = Time()
  n87 = VectorParameter() {group=FallBack, parameterName=FallB_Normal Noise Speed, defaultValue=<rgba 0.01 0.03 -0.05 0>}
  n88 = Multiply(
   A=Add(A=Multiply(A=@n86, B=@n87.RGB), B=@n42)
   B=ScalarParameter() {defaultValue=10, group=FallBack, parameterName=FallB_Normal Noise Tile_Rate}
  )
  n89 = VertexInterpolator(VS=PreSkinnedNormal())
  n90 = ScalarParameter() {group=FallBack, parameterName=FallB_ProjectionTransitionContrast}
  n91 = LinearInterpolate(
   A=GetMaterialAttributes(None=FunctionInput() {bUsePreviewValueAsDefault=true, inputName=Material, inputType=FunctionInput_MaterialAttributes}) {attributeGetTypes [#1]}.Normal
   B=Constant3Vector() {constant=<rgba 0 0 1 0>}
   Alpha=ScalarParameter() {group=FallBack, parameterName="FallB_Disable_Base_Normal 0=off 1=active", sortPriority=5}
  )
  n92 = LinearInterpolate(
   A=@n91
   B=Add(
    A=@n91
    B=MaterialFunctionCall!(
     TextureObject=@n84
     TextureSize=@n85
     RotYAxis
     RotZAxis
     RotXAxis
     PreSkinNormal=@n89.PS
     ProjectionTransitionContrast=@n90
     WorldPosition=@n88
    ) {materialFunction=/Game/Packages/Fortress_SharedMaterials/Base_Material_Functions/BaseCustomizationFunctions/PreSkinnedAlignedNormal}.XYZ Texture
   )
   Alpha=ScalarParameter() {group=FallBack, parameterName=FallB_Normal Noise_Stregth}
  )
  n93 = Constant3Vector() {constant=<rgba 0 0 1 0>}
  n94 = Subtract(A=WorldPosition().XYZ, B=ObjectPositionWS())
  n95 = ObjectRadius()
  n96 = ScalarParameter() {defaultValue=0.025, group=Noise, parameterName=WPOStrengthFromObjectRadius}
  n97 = Multiply(
   A=Multiply(A=Normalize(VectorInput=@n94), B=Constant3Vector() {constant=<rgba 1 1 0 0>})
   B=Add(A=ScalarParameter() {defaultValue=2, group=Noise, parameterName=WPOStrength}, B=Multiply(A=@n95, B=@n96) {constB=0.05})
  )
  n98 = Divide(A=ObjectRadius(), B=ScalarParameter() {defaultValue=0.1, group=Noise, parameterName=NoiseSize})
  n99 = Multiply(A=Time(), B=VectorParameter() {group=FallBack_Glitch, parameterName=NoisePanSpeed, defaultValue=<rgba -0.08 0.07 -0.2 0>}.RGB)
  n100 = MaterialFunctionCall(
   TextureObject=TextureObject() {samplerType=SAMPLERTYPE_LinearColor, texture=&T_NoisePack_LinearColor}
   TextureSize=Multiply(A=VectorParameter() {group=Noise, parameterName=NoiseTileMult, defaultValue=<rgba 1 1 0.1 1>}.RGB, B=@n98)
   WorldPosition=Add(A=WorldPosition().XYZ, B=Multiply(A=@n98, B=@n99))
  ) {materialFunction=&WorldAlignedTexture}
  n101 = Multiply(
   A=Multiply(A=Power(Base=@n100.XY Texture) {constExponent=4}, B=ScalarParameter() {defaultValue=2, group=FallBack_Glitch, parameterName=NoiseMaskMult}) {constB=20}
   B=@n36
  )
  n102 = ComponentMask(None=MaterialFunctionCall() {materialFunction=&ObjectPivotPoint}.Object Pivot Location) {b=true}
  n103 = ScalarParameter() {defaultValue=300, group=FallBack_Glitch, parameterName=PulseHeight}
  n104 = ScalarParameter() {defaultValue=1, group=FallBack_Glitch, parameterName=TurnOn}
  n105 = ScalarParameter() {defaultValue=33, group=FallBack_Glitch, parameterName=PulseRadius}
  n106 = SphereMask(
   A=ComponentMask(None=WorldPosition().XYZ) {b=true}
   B=LinearInterpolate(A=@n102, B=Add(A=@n102, B=@n103) {constB=1600}, Alpha=@n104) {constA=-0.25, constB=1.2}
   Radius=LinearInterpolate(A=@n105, B=Multiply(A=@n105) {constB=0.1}, Alpha=@n104) {constA=500, constB=100}
  ) {hardnessPercent=0}
  n107 = OneMinus(None=@n104)
  n108 = Add(A=Desaturation(None=Saturate(None=@n101)), B=Multiply(A=Multiply(A=@n106, B=@n107)) {constB=2})
  n109 = ScalarParameter() {defaultValue=1, group=FallBack, parameterName=FallB_Phase Mask Amount, sortPriority=-10}
  n110 = Add(
   A=@n42
   B=Multiply(A=Time(), B=VectorParameter() {group=FallBack, parameterName=FallB_Transition_Smoke_Speed, defaultValue=<rgba 0.01 0.01 -0.03 1>}.RGB)
  )
  n111 = ScalarParameter() {defaultValue=2, group=FallBack, parameterName=FallB_Transition_Smoke_Tile_Rate}
  n112 = ComponentMask(None=MaterialFunctionCall(
    TextureObject=@n48
    TextureSize=Constant3Vector() {constant=<rgba 1 1 1 0>}
    WorldPosition=Multiply(A=@n110, B=@n111)
    Export Float 4=StaticBool() {value=true}
    World Space Normal=@n89.PS
    ProjectionTransitionContrast=@n90
   ) {materialFunction=&WorldAlignedTexture}.XYZ Texture) {a=true}
  FunctionOutput(None=MakeMaterialAttributes!(
    BaseColor=@n1.RGB
    Metallic
    Specular=ScalarParameter() {defaultValue=0.5, group=FallBack, parameterName=FallB_Specular, sortPriority=1}
    Roughness=ScalarParameter() {defaultValue=0.781, group=FallBack, parameterName=FallB_Roughness, sortPriority=1}
    Anisotropy
    EmissiveColor=Add(A=@n78, B=@n83)
    Opacity
    OpacityMask=MaterialFunctionCall(Transition Phase=@n109, Height Texture=@n112) {materialFunction=&HeightLerp}.Alpha
    Normal=ShadingPathSwitch(Default=@n92, Deferred=@n92, Mobile=@n93) {desc=Migrated from FeatureLevelSwitch}
    Tangent
    WorldPositionOffset=Multiply(A=@n108, B=@n97) {constB=0}
    SubsurfaceColor
    ClearCoat
    ClearCoatRoughness
    AmbientOcclusion
    Refraction
    CustomizedUVs_0
    CustomizedUVs_1
    CustomizedUVs_2
    CustomizedUVs_3
    CustomizedUVs_4
    CustomizedUVs_5
    CustomizedUVs_6
    CustomizedUVs_7
    PixelDepthOffset
    ShadingModel
    Displacement
   ))
  OneMinus(None=@n15)
  OneMinus(None=@n17)
 ]}
