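# Fallback character material for skinned meshes under the generic smooth medium male player body, using masked, default-lit material attributes with morph target and mesh deformer support.
# The shared character PreFX function produces the base attributes, which pass through MF_FallBack into the shared PostFX function.
# The PreFX AO output also drives PostFX's AO, Distance Blend, Fresnel and Gmap inputs, rather than separate signals for each.
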Material /Game/Characters/Player/Male/Medium/Bodies/M_MED_Generic_Smooth/Materials/M_FN_Character_FallBack
graph {
 properties {
  bUseMaterialAttributes=true
  bUsedWithMeshDeformer=true
  bUsedWithMorphTargets=true
  bUsedWithSkeletalMesh=true
  blendMode=BLEND_Masked
  pixelDepthOffsetMode=PDOM_Legacy
  usedShadingModels=Default Lit
 }
 nodes [
  n1 = MaterialFunctionCall() {materialFunction=&BaseFN_CharacterFunction_PreFX}
  MaterialFunctionCall(
   AO=@n1.AO
   Distance Blend=@n1.AO
   Fresnel=@n1.AO
   Gmap=@n1.AO
   Material=MaterialFunctionCall!(Material=@n1.Result) {materialFunction=&/MF_FallBack}.Result
  ) {materialFunction=&BaseFN_CharacterFunction_PostFX}
 ]
}
